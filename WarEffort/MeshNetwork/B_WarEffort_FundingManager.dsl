(event EventBeginPlay
  ;[0 0.6094 1] Initialize
  (Initialization|InitializeChoiceData)
  (if (Networking|IsDedicatedServer)
    (bind _nodetype (Class|FortAsyncActionMeshNetworkReady|MeshNetworkReadyAsync)
      (:then
        (if (Variables|Debug|GetDebug_SkipMeshNetwork)
          (CallFunction|MeshNetworkReady "Edge")))
      (:OnReady
        (if (not (Variables|Debug|GetDebug_SkipMeshNetwork))
          (CallFunction|MeshNetworkReady _nodetype)
          (MeshNetworkSubsystem|BindEventtoOnMeshNodeTypeChanged (GameInstanceSubsystems|GetMeshNetworkSubsystem) (EventDispatchers|CreateEvent :!event Custom|MeshNetworkReady)))))
    (else
      (Class|FortAsyncActionMeshNetworkReady|MeshNetworkReadyAsync
        (:OnReady
          (Utilities|FlowControl|Delay 3.0)
          (Initialization|InitializeFundingStations))))))

(event Custom|ChangeTBYBCalendarEventState
  (params (bIsActive :type Boolean)))

(event Custom|MeshMetaDataUpdated
  (bind _indexedfundingdata (Utilities|Struct|BreakWarEffortFundingMetadata (Variables|MeshNetwork|GetFundingStationMetaData)))
  (bind _indexedfundingdata_currentfundingarray (Utilities|Struct|BreakWarEffortIndexedFundingData _indexedfundingdata))
  ;[1 1 1] Restore initial vote value if the root was restarted
  (bind _returnvalue (MeshNetworkSubsystem|GetMetaDatawithKey (GameInstanceSubsystems|GetMeshNetworkSubsystem) (Variables|MeshNetwork|GetMetaDataKey) (Variables|MeshNetwork|GetFundingStationMetaData)))
  (if _returnvalue
    (Variables|Funding|SetCurrentFunds _indexedfundingdata_currentfundingarray)
    (Funding|Set|UpdateMutatorFundingData)
    (CallFunction|RootSetFunds)
    (else
      (CallFunction|RootSetFunds))))

(event Custom|ClearAllCurrentFunds
  (bind _currentfunds (Variables|Funding|GetCurrentFunds))
  ;[1 1 1] [Root] Reset current funds for all indices to zero
  (Fortnite|FortShippingLog self "[ClearAllCurrentFunds] zeroing out all current funding data after mesh event received!")
  (for (_ _array_index) _currentfunds
    (Utilities|Array|SetArrayElem _currentfunds _array_index))
  (Funding|Set|UpdateMutatorFundingData)
  (CallFunction|RootSetFunds)
  ;;[1 1 1] Persist funding data to mesh network
  (bind _structout (Utilities|Struct|SetmembersinWarEffortFundingMetadata (Variables|MeshNetwork|GetFundingStationMetaData) (Utilities|Struct|MakeWarEffortIndexedFundingData (Variables|Funding|GetCurrentFunds) (Math|Float|TruncatetoInteger64 (ScalableFloat|GetValueatLevel (Variables|Config|GetFinalFundingLimit))) (Math|Float|TruncatetoInteger64 (ScalableFloat|GetValueatLevel (Variables|Config|GetTowerFundingLimit)))) (Variables|Funding|GetFundingChoices)))
  (MeshNetworkSubsystem|SetMetaDatawithKey (GameInstanceSubsystems|GetMeshNetworkSubsystem) (Variables|MeshNetwork|GetMetaDataKey) _structout))

(event Custom|RootSetFunds
  ;[1 1 1] Update Funding
  (bind _result (Math|Integer|IncrementInt (Variables|MeshNetwork|GetUpdateRootInt))
    (:%20%20
      (Variables|MeshNetwork|SetUpdateRootInt _result))))

(event Custom|UpdateFunds
  (params (FundsAdded :type Integer64) (Funding%20Station :type Integer) (FundingChoiceIndex :type Integer) (FundingIndex :type Integer) (ChoiceDataIndex :type Integer))
  ;[1 1 1] Update Funding
  (bind _new_total_funds (Funding|Set|SetFundsbyIndex FundsAdded FundingIndex))
  (Funding|Set|UpdateOptionFunds FundingChoiceIndex _new_total_funds ChoiceDataIndex)
  (CallFunction|RepTotalFunds _new_total_funds Funding%20Station FundingChoiceIndex)
  ;;[1 1 1] Persist funding data to mesh network
  (bind _structout (Utilities|Struct|SetmembersinWarEffortFundingMetadata (Variables|MeshNetwork|GetFundingStationMetaData) (Utilities|Struct|MakeWarEffortIndexedFundingData (Variables|Funding|GetCurrentFunds) (Math|Float|TruncatetoInteger64 (ScalableFloat|GetValueatLevel (Variables|Config|GetFinalFundingLimit))) (Math|Float|TruncatetoInteger64 (ScalableFloat|GetValueatLevel (Variables|Config|GetTowerFundingLimit)))) (Variables|Funding|GetFundingChoices)))
  (MeshNetworkSubsystem|SetMetaDatawithKey (GameInstanceSubsystems|GetMeshNetworkSubsystem) (Variables|MeshNetwork|GetMetaDataKey) _structout))

(event Custom|RepTotalFunds
  (params (NewFunds :type Integer64) (FundingStationIndex :type Integer) (Choice :type Integer))
  ;[1 1 1] Update Funding
  (Funding|Set|UpdateTotalFunds NewFunds FundingStationIndex Choice)
  (switch Utilities|FlowControl|Switch|SwitchonEMeshNetworkNodeType (MeshNetworking|GetMeshNetworkNodeType (Variables|Default|GetB_WarEffort_MeshNetworkComponent))
    (:Root
      (Funding|Set|UpdateMutatorFundingData))))

(event Custom|PrintFundingIndexInfo
  ;[1 1 1] DEBUG CHEAT: Print funding index info
  (CallFunction|DebugPrintFundingIndexInfo false))

(event Custom|MeshNetworkReady
  (params (NodeType :type EMeshNetworkNodeType%20Enum :default "Root"))
  ;[1 1 1] Root / Edge Nodes
  (Fortnite|FortShippingLog self (Utilities|String|ToString(Text) (Utilities|Text|FormatText "NSLOCTEXT(\"[E37BACA94E6F32855DC62BAE74821264]\", \"99773308464A60A4BF20BC8A63BDF1B8\", \"MeshNetworkReady: NodeType{0}\")" (Utilities|String|EnumtoString NodeType))))
  (switch Utilities|FlowControl|Switch|SwitchonEMeshNetworkNodeType NodeType
    (:Root
      (Initialization|InitializeFundingStations)
      (Initialization|InitializeFundingIndices)
      (Initiatives|UpdateActiveInitiatives)
      (CallFunction|CheckHotfixes)
      (CallFunction|CheckTryBeforeYouBuyEvents)
      (Funding|Set|UpdateAllFundingStations true)
      (Funding|Set|UpdateMutatorFundingData)
      (Default|BindEventtoOnFundsAdded (Variables|Default|GetB_WarEffort_MeshNetworkComponent) (EventDispatchers|CreateEvent :!event Custom|UpdateFunds))
      (MeshNetworkSubsystem|BindEventtoOnMeshMetaDataUpdated (GameInstanceSubsystems|GetMeshNetworkSubsystem) (EventDispatchers|CreateEvent :!event Custom|MeshMetaDataUpdated))
      (CallFunction|MeshMetaDataUpdated)
      (for _array_element (Variables|/WarEffort/Data/WarEffortConfigData.WarEffortConfigData_C|GetInitiatives (Variables|Config|GetWarEffortConfigData))
        (if (.HasTryBeforeYouBuyPeriod_22_40DCD3C94B1EDA68079D20BAF960B439 (Utilities|Struct|BreakFWarEffortInitiativeConfig _array_element))
          (bind _eventname (MCPContext|WatchCalendarEvent (.TBYBCalendarEvent_25_759DB4C942E3EDC3A00CB6B981AF5BB0 (Utilities|Struct|BreakFWarEffortInitiativeConfig _array_element)))
            (:OnEventActive
              (CallFunction|HandleCalendarEventChanged _eventname (Utilities|Struct|BreakFWarEffortInitiativeConfig _array_element) true))
            (:OnEventEnded
              (CallFunction|HandleCalendarEventChanged _eventname (Utilities|Struct|BreakFWarEffortInitiativeConfig _array_element))))))
      (Funding|Set|SetFundingManagerReadyforMutator))
    (:Inner
      (Funding|Set|SetFundingManagerReadyforMutator))
    (:Edge
      (Class|FortAsyncActionAuthorityMatchReady|MatchReadyAsync
        (:OnReady
          (Initialization|InitializeFundingStations)
          (Initialization|InitializeFundingIndices)
          (Initiatives|UpdateActiveInitiatives)
          (CallFunction|CheckHotfixes)
          (CallFunction|CheckTryBeforeYouBuyEvents)
          (Funding|Set|UpdateAllFundingStations true)
          (Funding|Set|UpdateMutatorFundingData)
          (for (_array_element_2 _array_index) (Variables|Funding|GetFundingStations)
            (Utilities|IsValid (Utilities|ResolveSoftReference _array_element_2)
              (:Is%20Valid
                (Default|BindEventtoOnFundingReceived (Utilities|ResolveSoftReference _array_element_2) (EventDispatchers|CreateEvent :!event Custom|FundingRecieved))
                (Funding|Set|SetFundingManagerReadyforMutator))
              (:Is%20Not%20Valid
                (Fortnite|FortShippingLog self (Utilities|String|ToString(Text) (Utilities|Text|FormatText "NSLOCTEXT(\"[E37BACA94E6F32855DC62BAE74821264]\", \"281D88E1480B12C20A8B01805DCD0F70\", \"FundingManagerError: Could not resolve station at index {0}!\")" _array_index)))))))))))

(event Custom|FundingRecieved
  (params (FundingAmountAdded :type Integer64) (FundingStationNum :type Integer) (FundingChoiceIndex :type Integer))
  (bind _fundingstations (Variables|Funding|GetFundingStations))
  (bind _output (Utilities|Array|Get(acopy) _fundingstations FundingStationNum))
  (bind _output_2 (Utilities|ResolveSoftReference _output))
  (bind _initiativechoicemap (Variables|Initiatives|GetInitiativeChoiceMap))
  ;[0.9647 0.9792 0.9683] On Funds Recieved
  (Utilities|FlowControl|Sequence
    (:then_1
      (bind (_indexfound _value) (Funding|Index|GetFundingIndexforStation _output_2 FundingChoiceIndex))
      (bind _newtotal (Funding|Set|SetCurrentFundsforStationbyIndex FundingStationNum FundingAmountAdded FundingChoiceIndex true))
      (bind (_hasactiveinitiative _initiativetag) (Helpers|GetInitiativeTagforStation _output_2))
      (if (not (and (Variables|Debug|GetDebug_SkipMeshNetwork) (Editor|InEditororPIE)))
        (Class|BWarEffortMeshNetworkComponent|FundingStation0 (Variables|Default|GetB_WarEffort_MeshNetworkComponent) FundingAmountAdded FundingStationNum FundingChoiceIndex _value (Math|Integer|SelectInt (Utilities|Map|Find _initiativechoicemap _initiativetag) -1 (.ReturnValue (Utilities|Map|Find _initiativechoicemap _initiativetag))))
        (else
          (CallFunction|UpdateFunds FundingAmountAdded FundingStationNum FundingChoiceIndex _value (Math|Integer|SelectInt (Utilities|Map|Find _initiativechoicemap _initiativetag) -1 (.ReturnValue (Utilities|Map|Find _initiativechoicemap _initiativetag)))))))))

(fn CheckCalendarEvent
  (params (EventName :type String))
  (if (Variables|Debug|GetUseDebugCalendarEvents)
    (return (Utilities|Array|ContainsItem (Variables|Debug|GetDebugActiveCalendarEvents) EventName))
    (else
      (bind _biseventactive (CalendarEvents|GetCalendarEventInformation EventName))
      (return _biseventactive))))

(fn CheckHotfixes
  (if (ScalableFloat|GetValueasBool (Variables|Default|GetHF_Enabled))
    ;[1 1 1] Check if each funding station should be enabled. Remove inactive funding stations from array so they don't get checked down the line, then destroy
    (for _array_element (Variables|Funding|GetFundingStations)
      (bind _returnvalue (Class|BWarEffortFundingStationBase|GetEnabledHotfix (Utilities|ResolveSoftReference _array_element)))
      (if (not _returnvalue)
        (Utilities|Array|Add (LocalVariables|GetStationsToDestroy) (Utilities|ResolveSoftReference _array_element))))
    (for _array_element_2 (LocalVariables|GetStationsToDestroy)
      (Utilities|Array|RemoveItem (Variables|Funding|GetFundingStations) (Utilities|ResolveSoftReference _array_element_2))
      (Actor|DestroyActor _array_element_2))
    (else
      ;[1 1 1] Destroy funding stations and manager
      (for _array_element_3 (Variables|Funding|GetFundingStations)
        (Actor|DestroyActor (Utilities|ResolveSoftReference _array_element_3)))
      (Actor|DestroyActor))))

(fn CheckTryBeforeYouBuyEvents
  (for _array_element (Variables|/WarEffort/Data/WarEffortConfigData.WarEffortConfigData_C|GetInitiatives (Variables|Config|GetWarEffortConfigData))
    ;[1 1 1] For each initiative with a TBYB period, check if the TBYB calendar event is active
    (if (.HasTryBeforeYouBuyPeriod_22_40DCD3C94B1EDA68079D20BAF960B439 (Utilities|Struct|BreakFWarEffortInitiativeConfig _array_element))
      (bind _biseventactive (CallFunction|CheckCalendarEvent (.TBYBCalendarEvent_25_759DB4C942E3EDC3A00CB6B981AF5BB0 (Utilities|Struct|BreakFWarEffortInitiativeConfig _array_element))))
      (if _biseventactive
        (Utilities|FlowControl|Sequence
          (:then_0
            (for _array_element_2 (.Options_6_F30346D942E5CB445FA46C93CAB109F2 (Utilities|Struct|BreakFWarEffortInitiativeConfig _array_element))
              (Default|EnableTryBeforeYouBuyforItem (Utilities|Struct|BreakFWarEffortOptionConfig _array_element_2))))
          (:then_1
            (bind _outstations (CallFunction|GetAllStationsforInitiative (Utilities|Struct|BreakFWarEffortInitiativeConfig _array_element)))
            (for _array_element_3 _outstations
              (Variables|/WarEffort/Stations/B_WarEffort_FundingStation_Base.B_WarEffort_FundingStation_Base_C|SetMonth (.VoteStartMonth_30_CDE7CC764ECAD8DAFE0CDE9E339F7EB1 (Utilities|Struct|BreakFWarEffortInitiativeConfig _array_element)) _array_element_3)
              (Variables|/WarEffort/Stations/B_WarEffort_FundingStation_Base.B_WarEffort_FundingStation_Base_C|SetDay (.VoteStartDay_32_8D141FEF4BA27EA068D89D9A1D8CF29A (Utilities|Struct|BreakFWarEffortInitiativeConfig _array_element)) _array_element_3)
              (Variables|/WarEffort/Stations/B_WarEffort_FundingStation_Base.B_WarEffort_FundingStation_Base_C|SetStage 1 _array_element_3))))))))

(fn Debug_PrintFundingIndexInfo
  (params (OnlyActiveInitiatives :type Boolean))
  (bind _currentchoice (LocalVariables|GetCurrentChoice))
  (bind _currentfundingstationindex (LocalVariables|GetCurrentFundingStationIndex))
  (bind _currentinitiativetag (LocalVariables|GetCurrentInitiativeTag))
  (bind _returnvalue (GameplayTags|GetDebugStringfromGameplayTag _currentinitiativetag))
  (bind _currentindex (LocalVariables|GetCurrentIndex))
  (bind _result (Utilities|Text|FormatText "NSLOCTEXT(\"[E37BACA94E6F32855DC62BAE74821264]\", \"1C8B3D514694945CCDAF6C9AF96C50EA\", \"Index: {0} | Initiative: {1} | Funding Station: {2} | Choice: {3}\")" _currentindex _returnvalue _currentfundingstationindex _currentchoice))
  (bind _returnvalue_2 (Utilities|String|ToString(Text) _result))
  (bind _currentinitiativetag_2 (LocalVariables|GetCurrentInitiativeTag))
  (bind _currentinitiativetag_3 (LocalVariables|GetCurrentInitiativeTag))
  (bind _output_get (LocalVariables|SetInitiativesToPrint (Variables|/WarEffort/Data/WarEffortConfigData.WarEffortConfigData_C|GetInitiatives (Variables|Config|GetWarEffortConfigData))))
  (for _array_element _output_get
    (LocalVariables|SetCurrentInitiativeTag (Utilities|Struct|BreakFWarEffortInitiativeConfig _array_element))
    (if (.TrackFundsPerStation_19_8D8773804538AB8EC192EABCFD9AE2C2 (Utilities|Struct|BreakFWarEffortInitiativeConfig _array_element))
      (bind _outstations (CallFunction|GetAllStationsforInitiative (LocalVariables|GetCurrentInitiativeTag)))
      (for _array_element_2 _outstations
        (LocalVariables|SetCurrentFundingStationIndex (Variables|/WarEffort/Stations/B_WarEffort_FundingStation_Base.B_WarEffort_FundingStation_Base_C|GetFundingStationNum _array_element_2))
        (for (_ _array_index) (.Options_6_F30346D942E5CB445FA46C93CAB109F2 (Utilities|Struct|BreakFWarEffortInitiativeConfig _array_element))
          (LocalVariables|SetCurrentChoice _array_index)
          (LocalVariables|SetCurrentIndex (Funding|Index|GetFundsIndexforInitiative _currentinitiativetag_3 _array_index _array_element_2))
          (if (.FoundIndex (Funding|Index|GetFundsIndexforInitiative _currentinitiativetag_3 _array_index _array_element_2))
            (Development|PrintString _returnvalue_2 :!label _label)
            (Fortnite|FortShippingLog self _returnvalue_2))))
      (else
        (LocalVariables|SetCurrentFundingStationIndex -1)
        (for (_ _array_index_2) (.Options_6_F30346D942E5CB445FA46C93CAB109F2 (Utilities|Struct|BreakFWarEffortInitiativeConfig _array_element))
          (LocalVariables|SetCurrentChoice _array_index_2)
          (LocalVariables|SetCurrentIndex (Funding|Index|GetFundsIndexforInitiative _currentinitiativetag_2 _array_index_2))
          (if (.FoundIndex (Funding|Index|GetFundsIndexforInitiative _currentinitiativetag_2 _array_index_2))
            (goto _label)))))
    (goto _label)))

(fn Debug_SkipMeshNetwork_UpdateFunds
  (params (Funds :type Integer64) (Funding%20Station :type Integer) (FundingChoiceIndex :type Integer))
  (if (and (Variables|Debug|GetDebug_SkipMeshNetwork) (Editor|InEditororPIE))
    (CallFunction|UpdateFunds Funds Funding%20Station FundingChoiceIndex)))

(fn DoesInitiativeTrackFundsPerStation
  (params (InitiativeTag :type Gameplay%20Tag%20Structure))
  (bind (_initiativefound _initiative) (Class|WarEffortConfigData|GetInitiativebyTag (Variables|Config|GetWarEffortConfigData) InitiativeTag))
  (if _initiativefound
    (return (.TrackFundsPerStation_19_8D8773804538AB8EC192EABCFD9AE2C2 (Utilities|Struct|BreakFWarEffortInitiativeConfig _initiative)))
    (else
      (return false))))

(fn EnableTryBeforeYouBuyForItem
  (params (Item%20To%20Enable :type Gameplay%20Tag%20Structure))
  (bind _self self)
  (bind _wareffortmutatorref (Variables|Mutator|GetWarEffortMutatorRef))
  (Utilities|FlowControl|Sequence
    (:then_0
      ;[1 1 1] Make sure we have a valid mutator reference
      (Utilities|IsValid (Variables|Mutator|GetWarEffortMutatorRef)
        (:Is%20Not%20Valid
          (bind _asfort_game_state_athena (Utilities|Casting|CastToFortGameStateAthena (Game|GetGameState))
            (:then
              (Utilities|IsValid (Modifier|GetMutatorbyClass (Variables|/Script/FortniteGame.FortGameStateAthena|GetMutatorListComponent _asfort_game_state_athena) "/Script/CorruptionGameplayCodeRuntime.FortAthenaMutator_WarEffort")
                (:Is%20Valid
                  (bind _asfort_athena_mutator_war_effort (Utilities|Casting|CastToFortAthenaMutator_WarEffort (Modifier|GetMutatorbyClass (Variables|/Script/FortniteGame.FortGameStateAthena|GetMutatorListComponent _asfort_game_state_athena) "/Script/CorruptionGameplayCodeRuntime.FortAthenaMutator_WarEffort"))
                    (:then
                      (Variables|Mutator|SetWarEffortMutatorRef _asfort_athena_mutator_war_effort))
                    (:CastFailed
                      (Fortnite|FortShippingLog _self "[EnableTBYB] failed cast to WarEffort mutator!" true))))
                (:Is%20Not%20Valid
                  (Fortnite|FortShippingLog _self "[EnableTBYB] mutator ref from GSA was not valid!" true))))
            (:CastFailed
              (Fortnite|FortShippingLog self "[EnableTBYB] GSA cast failed" true))))))
    (:then_1
      (Utilities|IsValid _wareffortmutatorref
        (:Is%20Valid
          (Class|FortAthenaMutatorWarEffort|SetTryBeforeYouBuyItemState _wareffortmutatorref Item%20To%20Enable true))
        (:Is%20Not%20Valid
          (Fortnite|FortShippingLog self (Utilities|String|ToString(Text) (Utilities|Text|FormatText "NSLOCTEXT(\"[E37BACA94E6F32855DC62BAE74821264]\", \"50E3D662483847E8D8468494678B3FE1\", \"[EnableTBYB] couldn\\'t get a valid ref to the mutator!\")")) true))))))

(fn GetAllInitiatives
  (return (Variables|/WarEffort/Data/WarEffortConfigData.WarEffortConfigData_C|GetInitiatives (Variables|Config|GetWarEffortConfigData))))

(fn GetAllStationsForInitiative
  (params (Tag :type Gameplay%20Tag%20Structure))
  (bind _fundingstations (Variables|Funding|GetFundingStations))
  (if (> (Utilities|Array|Length _fundingstations) 0)
    (for _array_element _fundingstations
      (Utilities|IsValid (Utilities|ResolveSoftReference _array_element)
        (:Is%20Valid
          (if (GameplayTags|HasTag (Variables|/WarEffort/Stations/B_WarEffort_FundingStation_Base.B_WarEffort_FundingStation_Base_C|GetInitiatives (Utilities|ResolveSoftReference _array_element)) Tag true)
            (Utilities|Array|Add (LocalVariables|GetStations) (Utilities|ResolveSoftReference _array_element))))))
    (return (LocalVariables|GetStations))))

(fn GetCurrentFundsForInitiative
  (params (InitiativeTag :type Gameplay%20Tag%20Structure) (ChoiceIndex :type Integer) (Station :type B%20War%20Effort%20Funding%20Station%20Base%20Object%20Reference))
  (bind (_index _foundindex) (Funding|Index|GetFundsIndexforInitiative InitiativeTag ChoiceIndex Station))
  (if _foundindex
    (return (Utilities|Array|Get(acopy) (Variables|Funding|GetCurrentFunds) _index))))

(fn GetCurrentFundsForStation
  (params (FundingStation :type B%20War%20Effort%20Funding%20Station%20Base%20Object%20Reference) (Choice :type Integer))
  (bind _output (Funding|Get|GetCurrentFundsforInitiative (Variables|/WarEffort/Stations/B_WarEffort_FundingStation_Base.B_WarEffort_FundingStation_Base_C|GetActiveInitiative FundingStation) Choice FundingStation))
  (return _output))

(fn GetCurrentFundsForStationByIndex
  (params (StationIndex :type Integer) (Choice :type Integer))
  (bind _funds (Funding|Get|GetCurrentFundsforStation (Utilities|ResolveSoftReference (Utilities|Array|Get(acopy) (Variables|Funding|GetFundingStations) StationIndex)) Choice))
  (return _funds))

(fn GetFundingIndexForStation
  (params (Station :type B%20War%20Effort%20Funding%20Station%20Base%20Object%20Reference) (Choice :type Integer))
  (bind (_index _foundindex) (Funding|Index|GetFundsIndexforInitiative (Variables|/WarEffort/Stations/B_WarEffort_FundingStation_Base.B_WarEffort_FundingStation_Base_C|GetActiveInitiative Station) Choice Station))
  (return _foundindex _index))

(fn GetFundingTargetByInitiative
  (params (InitiativeTag :type Gameplay%20Tag%20Structure))
  (bind _wareffortconfigdata (Variables|Config|GetWarEffortConfigData))
  (bind (_initiativefound _initiative) (Class|WarEffortConfigData|GetInitiativebyTag _wareffortconfigdata InitiativeTag))
  (if _initiativefound
    (bind (_fundinglimitfound _fundinglimit) (Class|WarEffortConfigData|GetFundingLimitbyFundingType _wareffortconfigdata (.InitiativeTypeTag_14_11CACBAB4F570F662FA860AC59CC19F3 (Utilities|Struct|BreakFWarEffortInitiativeConfig _initiative))))
    (return _fundinglimitfound _fundinglimit)
    (else
      (return false 0.0))))

(fn GetFundsIndexForInitiative
  (params (InitiativeTag :type Gameplay%20Tag%20Structure%20%28by%20ref%29) (Choice :type Integer) (Station :type B%20War%20Effort%20Funding%20Station%20Base%20Object%20Reference))
  (bind _initiativefundingindices (Variables|Initiatives|GetInitiativeFundingIndices))
  (bind (_value _returnvalue) (Utilities|Map|Find _initiativefundingindices InitiativeTag))
  (bind _trackfundsperstation (CallFunction|DoesInitiativeTrackFundsPerStation InitiativeTag))
  (if _trackfundsperstation
    (bind (_indexfound _index) (Initiative|GetIndexforInitiative (Variables|GetStation) (Variables|GetInitiativeTag)))
    (LocalVariables|SetLocalFoundIndex _indexfound)
    (LocalVariables|SetStartingIndex _index)
    (else
      (LocalVariables|SetStartingIndex _value)
      (LocalVariables|SetLocalFoundIndex _returnvalue)))
  (return (+ (LocalVariables|GetStartingIndex) (Variables|GetChoice)) (LocalVariables|GetLocalFoundIndex)))

(fn GetInitiativeAtFundingIndex
  (params (Index :type Integer))
  (bind _values (Utilities|Map|Values (Variables|Initiatives|GetInitiativeFundingIndices)))
  (bind _output_get (LocalVariables|SetIndexMapValues _values))
  (while false
    _))

(fn GetInitiativeTagForStation
  (params (Station :type B%20War%20Effort%20Funding%20Station%20Base%20Object%20Reference))
  (bind (_hasactiveinitiative _initiativetag) (Initiative|FindActiveInitiative Station (GameplayTags|MakeGameplayTagContainerfromArray (Variables|Initiatives|GetActiveInitiatives))))
  (return _hasactiveinitiative _initiativetag))

(fn GetNumStationsForInitiative
  (params (InitiativeTag :type Gameplay%20Tag%20Structure))
  (bind _outstations (CallFunction|GetAllStationsforInitiative InitiativeTag))
  (return (Utilities|Array|Length _outstations)))

(fn HandleCalendarEventChanged
  (params (Event%20Name :type String) (Item%20Tag :type Gameplay%20Tag%20Structure) (Enabled :type Boolean))
  (bind _self self)
  (bind _self_2 self)
  (bind _wareffortmutatorref (Variables|Mutator|GetWarEffortMutatorRef))
  (Utilities|FlowControl|Sequence
    (:then_0
      ;[1 1 1] Make sure we have a valid mutator reference
      (Utilities|IsValid (Variables|Mutator|GetWarEffortMutatorRef)
        (:Is%20Not%20Valid
          (bind _asfort_game_state_athena (Utilities|Casting|CastToFortGameStateAthena (Game|GetGameState))
            (:then
              (Utilities|IsValid (Modifier|GetMutatorbyClass (Variables|/Script/FortniteGame.FortGameStateAthena|GetMutatorListComponent _asfort_game_state_athena) "/Script/CorruptionGameplayCodeRuntime.FortAthenaMutator_WarEffort")
                (:Is%20Valid
                  (bind _asfort_athena_mutator_war_effort (Utilities|Casting|CastToFortAthenaMutator_WarEffort (Modifier|GetMutatorbyClass (Variables|/Script/FortniteGame.FortGameStateAthena|GetMutatorListComponent _asfort_game_state_athena) "/Script/CorruptionGameplayCodeRuntime.FortAthenaMutator_WarEffort"))
                    (:then
                      (Variables|Mutator|SetWarEffortMutatorRef _asfort_athena_mutator_war_effort))
                    (:CastFailed
                      (Fortnite|FortShippingLog _self_2 "[SetFundingManagerReady] failed cast to WarEffort mutator!" true))))
                (:Is%20Not%20Valid
                  (Fortnite|FortShippingLog _self_2 "[SetFundingManagerReady] mutator ref from GSA was not valid!" true))))
            (:CastFailed
              (Fortnite|FortShippingLog self "[HandleCalendarEventChanged] GSA cast failed" true))))))
    (:then_1
      (Utilities|IsValid _wareffortmutatorref
        (:Is%20Valid
          (for _array_element (Variables|/WarEffort/Data/WarEffortConfigData.WarEffortConfigData_C|GetInitiatives (Variables|Config|GetWarEffortConfigData))
            (if (== (Utilities|Struct|BreakFWarEffortInitiativeConfig _array_element) Item%20Tag)
              (for _array_element_2 (.Options_6_F30346D942E5CB445FA46C93CAB109F2 (Utilities|Struct|BreakFWarEffortInitiativeConfig _array_element))
                (Class|FortAthenaMutatorWarEffort|SetTryBeforeYouBuyItemState _wareffortmutatorref (Utilities|Struct|BreakFWarEffortOptionConfig _array_element_2) Enabled)
                (Fortnite|FortShippingLog _self (Utilities|String|ToString(Text) (Utilities|Text|FormatText "NSLOCTEXT(\"[E37BACA94E6F32855DC62BAE74821264]\", \"6C6C1F3247F1E658A0B1DCBF2DF682C0\", \"[SetFundingManagerReady] TBYB item {0} state {1} set from calendar event!\")" (.DisplayName_22_E10681B54C5D823F36E65599A51952E5 (Utilities|Struct|BreakFWarEffortOptionConfig _array_element_2)) Enabled)))))))
        (:Is%20Not%20Valid
          (Fortnite|FortShippingLog _self (Utilities|String|ToString(Text) (Utilities|Text|FormatText "NSLOCTEXT(\"[E37BACA94E6F32855DC62BAE74821264]\", \"5D2C20264A3669F9B2590698F753A854\", \"[HandleCalendarEventChanged] couldn\\'t get a valid ref to the mutator!\")")) true))))))

(fn HasEnoughFunds
  (params (CurrentFunds :type Integer64) (MaxFunds :type Float%20%28double-precision%29%20%28by%20ref%29))
  (return (>= CurrentFunds (Math|Float|TruncatetoInteger64 MaxFunds))))

(fn InitializeChoiceData
  (bind _wareffortconfigdata (Variables|Config|GetWarEffortConfigData))
  (for _array_element (Variables|/WarEffort/Data/WarEffortConfigData.WarEffortConfigData_C|GetInitiatives _wareffortconfigdata)
    (bind (_fundinglimitfound _fundinglimit) (Class|WarEffortConfigData|GetFundingLimitbyFundingType _wareffortconfigdata (.InitiativeTypeTag_14_11CACBAB4F570F662FA860AC59CC19F3 (Utilities|Struct|BreakFWarEffortInitiativeConfig _array_element))))
    (if _fundinglimitfound
      (LocalVariables|SetFundingLimit (Math|Float|TruncatetoInteger64 _fundinglimit))
      (if (GameplayTags|HasTag (Variables|Mutator|GetMutatorRelevantFundingTypes) (.InitiativeTypeTag_14_11CACBAB4F570F662FA860AC59CC19F3 (Utilities|Struct|BreakFWarEffortInitiativeConfig _array_element)) true)
        ;[1 1 1] Choice Data currently has a hard two options, so we need to check if there actually is a valid second option before setting the tag. Some initiatives may be "mutator relevant" (i.e. need to affect the loot pool) but only have one option
        (LocalVariables|SetOption1Tag (select (Utilities|Array|IsValidIndex (.Options_6_F30346D942E5CB445FA46C93CAB109F2 (Utilities|Struct|BreakFWarEffortInitiativeConfig _array_element))) (Utilities|Struct|BreakFWarEffortOptionConfig (Utilities|Array|Get(acopy) (.Options_6_F30346D942E5CB445FA46C93CAB109F2 (Utilities|Struct|BreakFWarEffortInitiativeConfig _array_element)))) 0))
        (LocalVariables|SetOption2Tag (select (Utilities|Array|IsValidIndex (.Options_6_F30346D942E5CB445FA46C93CAB109F2 (Utilities|Struct|BreakFWarEffortInitiativeConfig _array_element)) 1) (Utilities|Struct|BreakFWarEffortOptionConfig (Utilities|Array|Get(acopy) (.Options_6_F30346D942E5CB445FA46C93CAB109F2 (Utilities|Struct|BreakFWarEffortInitiativeConfig _array_element)) 1)) 0))
        (bind _returnvalue (Utilities|Array|Add (Variables|Funding|GetFundingChoices) (Utilities|Struct|MakeWarEffortFundingChoiceData (Utilities|Struct|MakeWarEffortFundingOptionData (LocalVariables|GetOption1Tag)) (Utilities|Struct|MakeWarEffortFundingOptionData (LocalVariables|GetOption2Tag)) (LocalVariables|GetFundingLimit))))
        (Utilities|Map|Add (Variables|Initiatives|GetInitiativeChoiceMap) (Utilities|Struct|BreakFWarEffortInitiativeConfig _array_element) _returnvalue)))))

(fn InitializeFundingIndices
  (bind _currentindex (LocalVariables|GetCurrentIndex))
  (bind _currentindex_2 (LocalVariables|GetCurrentIndex))
  (bind _currentfunds (Variables|Funding|GetCurrentFunds))
  (Utilities|Map|Clear (Variables|Initiatives|GetInitiativeFundingIndices))
  (LocalVariables|SetCurrentIndex 0)
  (for _array_element (Variables|/WarEffort/Data/WarEffortConfigData.WarEffortConfigData_C|GetInitiatives (Variables|Config|GetWarEffortConfigData))
    (Utilities|Map|Add (Variables|Initiatives|GetInitiativeFundingIndices) (Utilities|Struct|BreakFWarEffortInitiativeConfig _array_element) _currentindex)
    ;[1 1 1] Cache the index we started at
    (LocalVariables|SetLastIndex _currentindex)
    (if (.TrackFundsPerStation_19_8D8773804538AB8EC192EABCFD9AE2C2 (Utilities|Struct|BreakFWarEffortInitiativeConfig _array_element))
      ;[1 1 1] If this initiative should track the funds of each station separately, iterate through all funding stations assigned to this initiative (note: this can be funding stations that aren't actively funding this initiative, but will fund it in the future), set its relative index for this initiative for ease of lookup, and increment current index based on the number of options
      (bind _outstations (CallFunction|GetAllStationsforInitiative (Utilities|Struct|BreakFWarEffortInitiativeConfig _array_element)))
      (bind _output_get (LocalVariables|SetInitiativeStations _outstations))
      (for _array_element_2 _output_get
        (Initiative|SetIndexforInitiative _array_element_2 (Utilities|Struct|BreakFWarEffortInitiativeConfig _array_element) _currentindex_2)
        (Development|PrintString (Utilities|String|ToString(Text) (Utilities|Text|FormatText "NSLOCTEXT(\"[E37BACA94E6F32855DC62BAE74821264]\", \"44AC7E724ED8F0CF28DA5DB1CB91EE05\", \"Index: {0} Initiative: {1} Choice: {2} Funding Station: {3}\")" _currentindex_2 (GameplayTags|GetDebugStringfromGameplayTag (Utilities|Struct|BreakFWarEffortInitiativeConfig _array_element)) 0 _array_element_2)))
        (LocalVariables|SetCurrentIndex (+ (LocalVariables|GetCurrentIndex) (Utilities|Array|Length (.Options_6_F30346D942E5CB445FA46C93CAB109F2 (Utilities|Struct|BreakFWarEffortInitiativeConfig _array_element))))))
      (else
        (Development|PrintString (Utilities|String|ToString(Text) (Utilities|Text|FormatText "NSLOCTEXT(\"[E37BACA94E6F32855DC62BAE74821264]\", \"44AC7E724ED8F0CF28DA5DB1CB91EE05\", \"Index: {0} Initiative: {1} Choice: {2} Funding Station: {3}\")" (LocalVariables|GetCurrentIndex) (GameplayTags|GetDebugStringfromGameplayTag (Utilities|Struct|BreakFWarEffortInitiativeConfig _array_element)) 0 0)))
        ;[1 1 1] If all funding stations will share the same funding amount for this initiative, simply increment the current index by the number of options
        (LocalVariables|SetCurrentIndex (+ (LocalVariables|GetCurrentIndex) (Utilities|Array|Length (.Options_6_F30346D942E5CB445FA46C93CAB109F2 (Utilities|Struct|BreakFWarEffortInitiativeConfig _array_element)))))))
    ;[1 1 1] After all indices for this initiative have been determined, iterate from the first index relevant to this initiative to the last one and populate the funding array with 0's if it doesn't currently have space for them
    (for _index (range (LocalVariables|GetLastIndex) (+ (LocalVariables|GetCurrentIndex) 1))
      (if (not (Utilities|Array|IsValidIndex _currentfunds _index))
        (Utilities|Array|SetArrayElem _currentfunds _index 0 true)))))

(fn InitializeFundingStations
  (bind _fundingstations (Variables|Funding|GetFundingStations))
  (Utilities|Array|Clear (Variables|Funding|GetFundingStations))
  (bind _outactors (Actor|GetAllActorsOfClass "/WarEffort/Stations/B_WarEffort_FundingStation_Base.B_WarEffort_FundingStation_Base_C"))
  (for _array_element _outactors
    (Variables|/WarEffort/Stations/B_WarEffort_FundingStation_Base.B_WarEffort_FundingStation_Base_C|SetFundingManagerRef self _array_element)
    (Utilities|Array|SetArrayElem _fundingstations (Variables|/WarEffort/Stations/B_WarEffort_FundingStation_Base.B_WarEffort_FundingStation_Base_C|GetFundingStationNum _array_element) (Utilities|ResolveSoftReference _array_element) true)
    (Fortnite|FortShippingLog self (Utilities|String|ToString(Text) (Utilities|Text|FormatText "NSLOCTEXT(\"[E37BACA94E6F32855DC62BAE74821264]\", \"0EB674494FD21FA35E7410B09AC765F0\", \"[InitializeFundingStations] found and initialized {0} stations\")" (Utilities|Array|Length _fundingstations))))))

(fn IsFundingStationFunded
  (params (Station :type B%20War%20Effort%20Funding%20Station%20Base%20Object%20Reference) (Choice :type Integer))
  (bind _funds (Funding|Get|GetCurrentFundsforStation Station Choice))
  (LocalVariables|SetCurrFund _funds)
  (bind (_hasactiveinitiative _initiativetag) (Helpers|GetInitiativeTagforStation Station))
  (if _hasactiveinitiative
    (bind (_fundinglimitfound _fundinglimit) (Funding|Get|GetFundingTargetbyInitiative _initiativetag))
    (if _fundinglimitfound
      (return (Helpers|HasEnoughFunds (LocalVariables|GetCurrFund) _fundinglimit))
      (else
        (return false)))
    (else
      (return false))))

(fn OnRep_UpdateRootInt
  (switch Utilities|FlowControl|Switch|SwitchonEMeshNetworkNodeType (MeshNetworking|GetMeshNetworkNodeType (Variables|Default|GetB_WarEffort_MeshNetworkComponent))
    (:Edge
      (Funding|Set|UpdateAllFundingStations false))))

(fn SetCurrentFundsForInitiative
  (params (InitiativeTag :type Gameplay%20Tag%20Structure) (Funds :type Integer64) (Choice :type Integer) (AddToTotal :type Boolean) (Station :type B%20War%20Effort%20Funding%20Station%20Base%20Object%20Reference))
  (bind (_index _foundindex) (Funding|Index|GetFundsIndexforInitiative InitiativeTag Choice Station))
  (if _foundindex
    (bind _output_get (LocalVariables|SetFundingArrayIndex _index))
    (LocalVariables|SetNewFunds (select AddToTotal (+ (Variables|GetFunds) (Utilities|Array|Get(acopy) (Variables|Funding|GetCurrentFunds) _output_get)) (Variables|GetFunds)))
    (Utilities|Array|SetArrayElem (Variables|Funding|GetCurrentFunds) (LocalVariables|GetFundingArrayIndex) (LocalVariables|GetNewFunds) true)
    (Development|PrintString (Utilities|String|ToString(Text) (Utilities|Text|FormatText "NSLOCTEXT(\"[E37BACA94E6F32855DC62BAE74821264]\", \"FC64C6E844E4E3D204E79A9F6BA8A3E9\", \"Initiative {0} choice {1} funded. New amount: {2} at index {3}\")" (GameplayTags|GetDebugStringfromGameplayTag (Variables|GetInitiativeTag)) (Variables|GetChoice) (Utilities|String|ToString(Integer64) (LocalVariables|GetNewFunds)) (LocalVariables|GetFundingArrayIndex))))
    (return (LocalVariables|GetNewFunds))))

(fn SetCurrentFundsForStation
  (params (Station :type B%20War%20Effort%20Funding%20Station%20Base%20Object%20Reference) (Funds :type Integer64) (Choice :type Integer) (AddToTotal :type Boolean))
  (bind (_hasactiveinitiative _initiativetag) (Helpers|GetInitiativeTagforStation Station))
  (if _hasactiveinitiative
    (bind _newtotal (Funding|Set|SetCurrentFundsforInitiative _initiativetag Funds Choice AddToTotal Station))
    (return _newtotal)))

(fn SetCurrentFundsForStationByIndex
  (params (StationIndex :type Integer) (Funds :type Integer64) (Choice :type Integer) (AddToTotal :type Boolean))
  (bind _newtotal (Funding|Set|SetCurrentFundsforStation (Utilities|ResolveSoftReference (Utilities|Array|Get(acopy) (Variables|Funding|GetFundingStations) StationIndex)) Funds Choice AddToTotal))
  (return _newtotal))

(fn SetFundingManagerReadyForMutator
  (bind _self self)
  (bind _self_2 self)
  (bind _wareffortmutatorref (Variables|Mutator|GetWarEffortMutatorRef))
  (Utilities|FlowControl|Sequence
    (:then_0
      ;[1 1 1] Make sure we have a valid mutator reference
      (Utilities|IsValid (Variables|Mutator|GetWarEffortMutatorRef)
        (:Is%20Not%20Valid
          (bind _asfort_game_state_athena (Utilities|Casting|CastToFortGameStateAthena (Game|GetGameState))
            (:then
              (Utilities|IsValid (Modifier|GetMutatorbyClass (Variables|/Script/FortniteGame.FortGameStateAthena|GetMutatorListComponent _asfort_game_state_athena) "/Script/CorruptionGameplayCodeRuntime.FortAthenaMutator_WarEffort")
                (:Is%20Valid
                  (bind _asfort_athena_mutator_war_effort (Utilities|Casting|CastToFortAthenaMutator_WarEffort (Modifier|GetMutatorbyClass (Variables|/Script/FortniteGame.FortGameStateAthena|GetMutatorListComponent _asfort_game_state_athena) "/Script/CorruptionGameplayCodeRuntime.FortAthenaMutator_WarEffort"))
                    (:then
                      (Variables|Mutator|SetWarEffortMutatorRef _asfort_athena_mutator_war_effort))
                    (:CastFailed
                      (Fortnite|FortShippingLog _self_2 "[SetFundingManagerReady] failed cast to WarEffort mutator!" true))))
                (:Is%20Not%20Valid
                  (Fortnite|FortShippingLog _self_2 "[SetFundingManagerReady] mutator ref from GSA was not valid!" true))))
            (:CastFailed
              (Fortnite|FortShippingLog self "[SetFundingManagerReady] GSA cast failed" true))))))
    (:then_1
      (Utilities|IsValid _wareffortmutatorref
        (:Is%20Valid
          (Class|FortAthenaMutatorWarEffort|SetFundingManagerReady _wareffortmutatorref true)
          (Fortnite|FortShippingLog _self "[SetFundingManagerReady] mutator has been informed that we're ready!"))
        (:Is%20Not%20Valid
          (Fortnite|FortShippingLog _self (Utilities|String|ToString(Text) (Utilities|Text|FormatText "NSLOCTEXT(\"[E37BACA94E6F32855DC62BAE74821264]\", \"6D23D9624818B4A778D6E6B17237D559\", \"[SetFundingManagerReady] couldn\\'t get a valid ref to the mutator!\")")) true))))))

(fn SetFundsByIndex
  (params (Added%20Funds :type Integer64%20%28by%20ref%29) (Index :type Integer))
  (bind _currentfunds (Variables|Funding|GetCurrentFunds))
  (bind _returnvalue (Utilities|Array|IsValidIndex _currentfunds Index))
  (bind _output (Utilities|Array|Get(acopy) _currentfunds Index))
  (bind _returnvalue_2 (+ Added%20Funds _output))
  (bind _returnvalue_3 (select _returnvalue _returnvalue_2 _output))
  (Utilities|Array|SetArrayElem _currentfunds Index _returnvalue_3 true)
  (return _returnvalue_3))

(fn SetRoundOptionDataFromCurrentFunds
  (bind _initiativechoicemap (Variables|Initiatives|GetInitiativeChoiceMap))
  (for _array_element (Variables|Initiatives|GetActiveInitiatives)
    (if (.ReturnValue (Utilities|Map|Find _initiativechoicemap _array_element))
      (bind (_initiativefound _initiative) (Class|WarEffortConfigData|GetInitiativebyTag (Variables|Config|GetWarEffortConfigData) _array_element))
      (if _initiativefound
        (bind _output (Funding|Get|GetCurrentFundsforInitiative _array_element))
        (bind _output_2 (Funding|Get|GetCurrentFundsforInitiative _array_element 1))
        (bind _output_optiontag_8_e5b134fd4f88a3c4586b619b0a461d2f (Utilities|Struct|BreakFWarEffortOptionConfig (Utilities|Array|Get(acopy) (.Options_6_F30346D942E5CB445FA46C93CAB109F2 (Utilities|Struct|BreakFWarEffortInitiativeConfig _initiative)) 1)))
        (bind _output_optiontag_8_e5b134fd4f88a3c4586b619b0a461d2f_2 (Utilities|Struct|BreakFWarEffortOptionConfig (Utilities|Array|Get(acopy) (.Options_6_F30346D942E5CB445FA46C93CAB109F2 (Utilities|Struct|BreakFWarEffortInitiativeConfig _initiative)))))
        (Utilities|Struct|SetmembersinWarEffortFundingChoiceData (Utilities|Array|Get(aref) (Variables|Funding|GetFundingChoices) (Utilities|Map|Find _initiativechoicemap _array_element)) (Utilities|Struct|MakeWarEffortFundingOptionData _output_optiontag_8_e5b134fd4f88a3c4586b619b0a461d2f_2 _output) (Utilities|Struct|MakeWarEffortFundingOptionData _output_optiontag_8_e5b134fd4f88a3c4586b619b0a461d2f _output_2))))))

(fn UpdateTotalFunds
  (params (NewFunds :type Integer64) (FundingStationIndex :type Integer) (Choice :type Integer))
  (bind _returnvalue (Editor|InEditororPIE))
  (bind _b_wareffort_meshnetworkcomponent (Variables|Default|GetB_WarEffort_MeshNetworkComponent))
  (bind _returnvalue_2 (MeshNetworking|GetMeshNetworkNodeType _b_wareffort_meshnetworkcomponent))
  (bind _debug_skipmeshnetwork (Variables|Debug|GetDebug_SkipMeshNetwork))
  (bind _returnvalue_3 (and _debug_skipmeshnetwork _returnvalue))
  (bind _fundingstationindex (Variables|GetFundingStationIndex))
  (bind _fundingstations (Variables|Funding|GetFundingStations))
  (bind _output (Utilities|Array|Get(acopy) _fundingstations _fundingstationindex))
  (bind _output_2 (Utilities|ResolveSoftReference _output))
  (bind _returnvalue_4 (select _returnvalue_3 "Edge" _returnvalue_2))
  (if (> (Utilities|Array|Length (Variables|Funding|GetFundingStations)) 0)
    (switch Utilities|FlowControl|Switch|SwitchonEMeshNetworkNodeType _returnvalue_4
      (:Edge
        (if (not (Networking|IsDedicatedServer))
          (bind (_indexfound _value) (Funding|Index|GetFundingIndexforStation _output_2 Choice))
          (Funding|Set|UpdateUIFundingData _value (Variables|GetNewFunds)))
        (bind _output_get (LocalVariables|SetFundingStation _output_2))
        (Funding|Set|UpdateFundingStation _output_get))
      (:Client
        (Funding|Set|UpdateUIFundingData (Variables|GetFundingStationIndex) (Variables|GetNewFunds))))
    (else
      (switch Utilities|FlowControl|Switch|SwitchonEMeshNetworkNodeType _returnvalue_4))))

(fn UpdateUIFundingData
  (params (FundingStationIndex :type Integer) (NewCurrentFund :type Integer64))
  (if (not (Networking|IsDedicatedServer))
    (Messaging|BroadcastMessage (GameInstanceSubsystems|GetGameplayMessageRouter) "(TagName=\"EventMessage.Conversation.Funding.UpdateCurrentFunds\")" (Utilities|Struct|MakeItemFundingUpdatedFundingData FundingStationIndex NewCurrentFund) true 0)))

(fn UpdateActiveInitiatives
  (Utilities|Array|Clear (Variables|Initiatives|GetActiveInitiatives))
  (if (Variables|Debug|GetDebugAllInitiativesActive)
    ;[1 1 1] Set all initiatives as active initiatives if bDebugAllInitiativesActive is true
    (for _array_element (Variables|/WarEffort/Data/WarEffortConfigData.WarEffortConfigData_C|GetInitiatives (Variables|Config|GetWarEffortConfigData))
      (Utilities|Array|Add (Variables|Initiatives|GetActiveInitiatives) (Utilities|Struct|BreakFWarEffortInitiativeConfig _array_element)))
    (else
      (for _array_element_2 (Variables|/WarEffort/Data/WarEffortConfigData.WarEffortConfigData_C|GetInitiatives (Variables|Config|GetWarEffortConfigData))
        ;[1 1 1] Check killswitch first
        (if (ScalableFloat|GetValueasBool (.HF_InitiativeEnabled_35_9676546C4878919C3888D9B5F069F839 (Utilities|Struct|BreakFWarEffortInitiativeConfig _array_element_2)))
          ;[1 1 1] Add initiative's tag to the array if its calendar event is currently active
          (bind _biseventactive (CallFunction|CheckCalendarEvent (.CalendarEvent_13_0EC5947C4FF47940BA5D7D958C9E24BE (Utilities|Struct|BreakFWarEffortInitiativeConfig _array_element_2))))
          (if (or _biseventactive (ScalableFloat|GetValueasBool (.HF_IgnoreCalendarEvent_37_4C32E55742ED4D40B9DFAA80CC6BFBB2 (Utilities|Struct|BreakFWarEffortInitiativeConfig _array_element_2))))
            (Utilities|Array|Add (Variables|Initiatives|GetActiveInitiatives) (Utilities|Struct|BreakFWarEffortInitiativeConfig _array_element_2))
            (elif (.HasTryBeforeYouBuyPeriod_22_40DCD3C94B1EDA68079D20BAF960B439 (Utilities|Struct|BreakFWarEffortInitiativeConfig _array_element_2))
              (bind _biseventactive_2 (CallFunction|CheckCalendarEvent (.TBYBCalendarEvent_25_759DB4C942E3EDC3A00CB6B981AF5BB0 (Utilities|Struct|BreakFWarEffortInitiativeConfig _array_element_2))))
              (if _biseventactive_2
                (Utilities|Array|Add (Variables|Initiatives|GetTBYBInitiatives) (Utilities|Struct|BreakFWarEffortInitiativeConfig _array_element_2)))))))))
  (for _array_element_3 (Variables|Funding|GetFundingStations)
    ;[1 1 1] Check if the station is assigned to one of the active initiatives; sets the station's current active initiative if so, and deactivates (destroys) if not
    (bind _hasactiveinitiative (Initiative|SetActiveInitiative (Utilities|ResolveSoftReference _array_element_3) (GameplayTags|MakeGameplayTagContainerfromArray (Variables|Initiatives|GetActiveInitiatives))))
    (if _hasactiveinitiative
      ;[1 1 1] If the station has an active initiative, additionally check if it has an individual calendar event. If that calendar event is not active, deactivate the station
      (if (Utilities|String|EqualExactly(String) (Variables|/WarEffort/Stations/B_WarEffort_FundingStation_Base.B_WarEffort_FundingStation_Base_C|GetActiveCalendarEvent (Utilities|ResolveSoftReference _array_element_3)))
        (Class|BWarEffortFundingStationBase|SetActive (Utilities|ResolveSoftReference _array_element_3) true)
        (else
          (bind _biseventactive_3 (CallFunction|CheckCalendarEvent (Variables|/WarEffort/Stations/B_WarEffort_FundingStation_Base.B_WarEffort_FundingStation_Base_C|GetActiveCalendarEvent (Utilities|ResolveSoftReference _array_element_3))))
          (Class|BWarEffortFundingStationBase|SetActive (Utilities|ResolveSoftReference _array_element_3) (or _biseventactive_3 (ScalableFloat|GetValueasBool (Variables|/WarEffort/Stations/B_WarEffort_FundingStation_Base.B_WarEffort_FundingStation_Base_C|GetHF_IgnoreCalendarEvent (Utilities|ResolveSoftReference _array_element_3)))))))
      (else
        ;[1 1 1] If the station doesn't have an active initiative, check if it has an active TBYB event
        (bind (_returnvalue _initiativetag) (Class|BWarEffortFundingStationBase|AssignedtoAnyInitiative (Utilities|ResolveSoftReference _array_element_3) (GameplayTags|MakeGameplayTagContainerfromArray (Variables|Initiatives|GetTBYBInitiatives))))
        (if _returnvalue
          (bind (_initiativefound _initiative) (Class|WarEffortConfigData|GetInitiativebyTag (Variables|Config|GetWarEffortConfigData) _initiativetag))
          (if _initiativefound
            (Variables|/WarEffort/Stations/B_WarEffort_FundingStation_Base.B_WarEffort_FundingStation_Base_C|SetBillboardMaterial (.BillboardImage_41_676D5E0248BE888C254D06B382F6940B (Utilities|Struct|BreakFWarEffortInitiativeConfig _initiative)) (Utilities|ResolveSoftReference _array_element_3)))
          (else
            (Class|BWarEffortFundingStationBase|SetActive (Utilities|ResolveSoftReference _array_element_3) false)))))))

(fn UpdateAllFundingStations
  (params (IsMatchStart :type Boolean))
  (if (> (Utilities|Array|Length (Variables|Funding|GetFundingStations)) 0)
    (for _array_element (Variables|Funding|GetFundingStations)
      (Utilities|IsValid (Utilities|ResolveSoftReference _array_element)
        (:Is%20Valid
          (bind _output_get (LocalVariables|SetCurrentFundingStation (Utilities|ResolveSoftReference _array_element)))
          (Funding|Set|UpdateFundingStation _output_get (Variables|GetIsMatchStart)))))))

(fn UpdateFundingStation
  (params (FundingStation :type B%20War%20Effort%20Funding%20Station%20Base%20Object%20Reference) (IsMatchStart :type Boolean))
  (bind _ismatchstart (Variables|GetIsMatchStart))
  (bind _returnvalue (not _ismatchstart))
  (bind _wareffortconfigdata (Variables|Config|GetWarEffortConfigData))
  (bind _fundingstation (Variables|GetFundingStation))
  ;[1 1 1] Compare the current active initiatives to the initiatives this station is assigned to find if it is currently assigned to an active one
  (bind (_hasactiveinitiative _initiativetag) (Initiative|FindActiveInitiative FundingStation (GameplayTags|MakeGameplayTagContainerfromArray (Variables|Initiatives|GetActiveInitiatives))))
  (if _hasactiveinitiative
    ;[1 1 1] Check if station has been hotfixed to be funded. If not, continue with normal check
    (bind _returnvalue_2 (Class|BWarEffortFundingStationBase|GetFundedHotfix FundingStation))
    (if _returnvalue_2
      (Class|BWarEffortFundingStationBase|EdgeServerShowFunding FundingStation)
      (bind _returnvalue_3 (Class|BWarEffortFundingStationBase|GetChoiceHotfix FundingStation))
      (Class|BWarEffortFundingStationBase|FundedState FundingStation _returnvalue_3 (not (Variables|GetIsMatchStart)))
      (else
        ;[1 1 1] Get the initiative data from config
        (bind (_initiativefound _initiative) (Class|WarEffortConfigData|GetInitiativebyTag _wareffortconfigdata _initiativetag))
        (if _initiativefound
          ;[1 1 1] Determine funding target based on the initiative's funding type
          (bind (_fundinglimitfound _fundinglimit) (Class|WarEffortConfigData|GetFundingLimitbyFundingType _wareffortconfigdata (Utilities|Struct|BreakFWarEffortInitiativeConfig _initiative)))
          (if _fundinglimitfound
            (LocalVariables|SetFundingTarget _fundinglimit)
            ;[1 1 1] Check funding of each option. Notify station if it was funded and what choice was funded. Otherwise, display funds
            (bind (_array_element _array_index) (Utilities|Array|ForEachLoopwithBreak (.Options_6_F30346D942E5CB445FA46C93CAB109F2 (Utilities|Struct|BreakFWarEffortInitiativeConfig _initiative)) :!label _label)
              (:LoopBody
                (bind _output (Funding|Get|GetCurrentFundsforInitiative _initiativetag _array_index (Variables|GetFundingStation)))
                (if (Helpers|HasEnoughFunds _output (LocalVariables|GetFundingTarget))
                  (if (.PreventMidmatchCompletion_27_1970EEF64ADF87A56EDE3CAF8284DCC7 (Utilities|Struct|BreakFWarEffortInitiativeConfig _initiative))
                    (if (not (Variables|GetIsMatchStart))
                      (Class|BWarEffortFundingStationBase|EdgeServerShowFunding _fundingstation :!label _label_2)))
                  (if (.TrackFundsPerStation_19_8D8773804538AB8EC192EABCFD9AE2C2 (Utilities|Struct|BreakFWarEffortInitiativeConfig _initiative))
                    (Class|BWarEffortFundingStationBase|FundedState _fundingstation _array_index _returnvalue)
                    (goto _label Break)
                    (else
                      (bind _outstations (CallFunction|GetAllStationsforInitiative _initiativetag))
                      (for _array_element_2 _outstations
                        (Class|BWarEffortFundingStationBase|FundedState _array_element_2 _array_index _returnvalue)
                        (goto _label Break))))
                  (goto _label))
                (goto _label_2)))))))))

(fn UpdateFundingStationsByInitiative
  (params (InitiativeTag :type Gameplay%20Tag%20Structure) (SkipFundingStation :type B%20War%20Effort%20Funding%20Station%20Base%20Object%20Reference))
  (bind _outstations (CallFunction|GetAllStationsforInitiative InitiativeTag))
  (for _array_element _outstations
    (Funding|Set|UpdateFundingStation _array_element)))

(fn UpdateMutatorFundingData
  (bind _self self)
  (Utilities|FlowControl|Sequence
    (:then_0
      ;[1 1 1] Make sure we have a valid mutator reference
      (Utilities|IsValid (Variables|Mutator|GetWarEffortMutatorRef)
        (:Is%20Not%20Valid
          (bind _asfort_game_state_athena (Utilities|Casting|CastToFortGameStateAthena (Game|GetGameState))
            (:then
              (Utilities|IsValid (Modifier|GetMutatorbyClass (Variables|/Script/FortniteGame.FortGameStateAthena|GetMutatorListComponent _asfort_game_state_athena) "/Script/CorruptionGameplayCodeRuntime.FortAthenaMutator_WarEffort")
                (:Is%20Valid
                  (bind _asfort_athena_mutator_war_effort (Utilities|Casting|CastToFortAthenaMutator_WarEffort (Modifier|GetMutatorbyClass (Variables|/Script/FortniteGame.FortGameStateAthena|GetMutatorListComponent _asfort_game_state_athena) "/Script/CorruptionGameplayCodeRuntime.FortAthenaMutator_WarEffort"))
                    (:then
                      (Variables|Mutator|SetWarEffortMutatorRef _asfort_athena_mutator_war_effort))
                    (:CastFailed
                      (Fortnite|FortShippingLog _self "[UpdateMutatorFundingData] failed cast to WarEffort mutator!" true))))
                (:Is%20Not%20Valid
                  (Fortnite|FortShippingLog _self "[UpdateMutatorFundingData] mutator ref from GSA was not valid!" true))))
            (:CastFailed
              (Fortnite|FortShippingLog self "[UpdateMutatorFundingData] GSA cast failed" true))))))
    (:then_1
      (Utilities|IsValid (Variables|Mutator|GetWarEffortMutatorRef)
        (:Is%20Valid
          (for _array_element (Variables|/WarEffort/Data/WarEffortConfigData.WarEffortConfigData_C|GetInitiatives (Variables|Config|GetWarEffortConfigData))
            ;[1 1 1] Check if this initiative has a funding type that means it is tied to options in the mutator. At the moment this just means it is a multi-option item funding, but that may be expanded in the future
            (if (GameplayTags|HasTag (Variables|Mutator|GetMutatorRelevantFundingTypes) (.InitiativeTypeTag_14_11CACBAB4F570F662FA860AC59CC19F3 (Utilities|Struct|BreakFWarEffortInitiativeConfig _array_element)))
              ;[1 1 1] Get target and current funding of the initiative
              (bind (_fundinglimitfound _fundinglimit) (Funding|Get|GetFundingTargetbyInitiative (Utilities|Struct|BreakFWarEffortInitiativeConfig _array_element)))
              (if _fundinglimitfound
                (for (_array_element_2 _array_index) (.Options_6_F30346D942E5CB445FA46C93CAB109F2 (Utilities|Struct|BreakFWarEffortInitiativeConfig _array_element))
                  (bind _output (Funding|Get|GetCurrentFundsforInitiative (Utilities|Struct|BreakFWarEffortInitiativeConfig _array_element) _array_index))
                  ;[1 0.233 0] Pass funding data for this option to the mutator. *Make sure the option tags in the option config and mutator match*
                  (Class|FortAthenaMutatorWarEffort|SetItemFundedAmount (Variables|Mutator|GetWarEffortMutatorRef) (Utilities|Struct|BreakFWarEffortOptionConfig _array_element_2) _output (Math|Float|TruncatetoInteger64 _fundinglimit)))))))))))

(fn UpdateOptionFunds
  (params (ChoiceIndex :type Integer) (NewFunds :type Integer64) (ChoiceDataIndex :type Integer))
  (bind _fundingchoices (Variables|Funding|GetFundingChoices))
  (bind _output (Utilities|Array|Get(aref) _fundingchoices ChoiceDataIndex))
  (bind (_option1 _option2) (Utilities|Struct|BreakWarEffortFundingChoiceData _output))
  (bind _newfunds (Variables|GetNewFunds))
  (if (Utilities|Array|IsValidIndex (Variables|Funding|GetFundingChoices) ChoiceDataIndex)
    (switch int (Variables|GetChoiceIndex)
      (:0
        (Utilities|Struct|SetmembersinWarEffortFundingOptionData _option1 _newfunds))
      (:1
        (Utilities|Struct|SetmembersinWarEffortFundingOptionData _option2 _newfunds)))))

(fn ConstructionScript
  (|CallParentFunction))
