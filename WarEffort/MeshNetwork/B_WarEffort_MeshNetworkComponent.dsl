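; War Effort mesh network component Blueprint that announces when funds are added to a funding station.
; Its single custom event, FundingStation0, receives the 64-bit funds amount plus station, funding choice, funding and choice data indices.
; It passes all five values unchanged to the OnFundsAdded dispatcher, with no other logic in the graph.
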(event Custom|FundingStation0
  (params (FundsAdded :type Integer64) (FundingStation :type Integer) (FundingChoiceIndex :type Integer) (FundingIndex :type Integer) (ChoiceDataIndex :type Integer))
  (Default|CallOnFundsAdded FundsAdded FundingStation FundingChoiceIndex FundingIndex ChoiceDataIndex))
